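# Idle glow effect for the EvokeFind pickaxe, built from a single sprite emitter that bursts three persistent particles at fixed local offsets, one per spawn group.
# A waveform shifts their color between purple and light blue, while a curve on system age fades alpha in over the first second.
# Each particle passes the owner's Z axis, tilted around X by a per-group angle, to the material as a dynamic parameter with a constant fourth value.
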NiagaraSystem /BRCosmetics/Weapons/FORT_Melee/Pickaxe_EvokeFind/FX/NS_EvokeFind_IdleGlow
dependencies {usedDataInterfaces [&NiagaraDataInterfaceCurve], usedDynamicInputs [], usedModules [&SystemState]}
emitters {Glow {
  emitterUpdateScript [
   &EmitterState(
    Inactive Response=::NewEnumerator1
    Life Cycle Mode=::NewEnumerator1
    Loop Behavior=::NewEnumerator1
    Loop Duration Mode=::NewEnumerator1
    UseLoopDelay=false
   )
   &SpawnBurst_Instantaneous(Spawn Count=3, Use Spawn Probability=false, UseLoopCountLimit=false)
  ]
  particleSpawnScript [
   &FN_SetSpawnGroup(Number of Spawn Groups=3)
   &InitializeParticle(
    Color=<rgba 1 1 1 2>
    Lifetime=5
    Position Offset=&MergeHelper_VectorSwitch(0=<0 0 7.5>, 1=<1.1 -5 7.5>, 2=<-1.1 5 7.5>, 3=?, Index=$Particles.SpawnGroup, Num=::NewEnumerator1)
    Position Offset Coordinate Space=::Local
    Sprite Size Mode=::NewEnumerator4
    UsePositionOffset=true
    Write Material Random=true
   )
   &CameraOffset(Camera Offset Amount=10)
  ]
  particleUpdateScript [
   &ParticleState(
    Kill Particles When Lifetime Has Elapsed=false
    Let Infinitely Lived Particles Die When Emitter Deactivates=true
    Loop Particles Lifetime=true
   )
   /Niagara/Noise/Waveform(
    EnableExponent1=false
    GlobalClampMax=false
    GlobalClampMin=false
    UseInput1=false
    UsePhase1=false
    "Waveform [1]":ENiagara_Waveforms=::NewEnumerator12
    "[1] Frequency"=0.5
   )
   &ScaleColor(
    Scale Alpha=&FloatFromCurve(
     FloatCurve=DataInterface() {propertyValues {Curve {
        defaultValue=0
        keys [
         {arriveTangent=4.1785, interpMode=RCIM_Cubic, leaveTangent=4.1785, tangentMode=RCTM_User, time=0, value=0}
         {arriveTangent=1.1179, interpMode=RCIM_Cubic, leaveTangent=1.1179, tangentMode=RCTM_User, time=0.2042, value=0.6228}
         {arriveTangent=0.0224, interpMode=RCIM_Cubic, leaveTangent=0.0224, tangentMode=RCTM_User, time=1, value=1}
        ]
       }}}
     CurveIndex=$Engine.System.Age
     Scale Curve=0.25
    )
    Scale RGB=&MakeVectorFromLinearColorRGB(Color=&LerpLinearColors(StartColor=<rgba 0.7253 0.25 1 0.5>, EndColor=<rgba 0.2135 0.6982 1 0.5>, LerpFactor=$Particles.Waveform.Waveform))
    ScaleA=true
    ScaleRGB=true
   )
   /SetVariables(Particles.DynamicMaterialParameter:Vector4f=&MakeVector4FromVectorAndFloat(
     "Vector (RGB)"=&NormalizeVector(
      Vector To Normalize=&RotateVectorByQuaternion(
       Vector To Rotate=$Engine.Owner.SystemZAxis
       Quaternion=&MakeQuaternion(
        Quaternion From=0
        Axis=$Engine.Owner.SystemXAxis
        Angle=&MergeHelper_FloatSwitch(0=0, 1=25, 2=-25, 3=?, Index=$Particles.SpawnGroup, Num=::NewEnumerator1)
        From=?
        To=?
        Angle Type=::NewEnumerator0
        Coordinate Space=::Local
        Normalize Vectors=true
       )
      )
      Safe Normalize=true
      Safe Fallback Threshold=0
      Safe Fallback Vector=<0 0 0.001>
     )
     "Float (Alpha)"=25
    )) {isSetParameters=true}
  ]
  rendererClasses [&NiagaraSpriteRendererProperties]
  renderers [{rendererClass=&NiagaraSpriteRendererProperties}]
 }}
summary {
 systemName=NS_EvokeFind_IdleGlow
 emitters [{emitterName=Glow, rendererClasses [&NiagaraSpriteRendererProperties]}]
 userVariables []
}
