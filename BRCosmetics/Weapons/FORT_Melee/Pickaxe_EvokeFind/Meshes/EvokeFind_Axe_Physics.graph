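# Physics asset for the EvokeFind pickaxe mesh, covering the bones of its dangling tassel.
# Capsule bodies sit on a master bone and six tassel bones, linked in a serial chain of constraints.
# The joint at the master bone has tighter, asymmetric swing limits, while the lower joints share wider, equal swing, and every joint allows only slight twist.
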PhysicsAsset /BRCosmetics/Weapons/FORT_Melee/Pickaxe_EvokeFind/Meshes/EvokeFind_Axe_Physics
bodies {
 dyn_tassel_01 [{shapeType=Capsule, length=2, radius=1.5, shapeName=dyn_tassel_01_capsule, center=<1.5209 0 0>, rotation=<rot -90 0.0002 0>}]
 dyn_tassel_02 [{shapeType=Capsule, length=2, radius=1.5, shapeName=dyn_tassel_02_capsule, center=<0.4839 0 0>, rotation=<rot -90 0.0002 0>}]
 dyn_tassel_03 [{shapeType=Capsule, length=2, radius=1.5, shapeName=dyn_tassel_03_capsule, center=<0.5377 0 0>, rotation=<rot -90 0.0002 0>}]
 dyn_tassel_04 [{shapeType=Capsule, length=2, radius=1.5, shapeName=dyn_tassel_04_capsule, center=<0.5823 0 0>, rotation=<rot -90 0.0002 0>}]
 dyn_tassel_05 [{shapeType=Capsule, length=2, radius=1.5, shapeName=dyn_tassel_05_capsule, center=<0.5964 0 0>, rotation=<rot -90 0.0002 0>}]
 dyn_tassel_06 [{shapeType=Capsule, length=2, radius=1.5, shapeName=dyn_tassel_06_capsule, center=<1.2555 0 0>, rotation=<rot -90 0.0002 0>}]
 dyn_tassel_MSTR [{shapeType=Capsule, length=0.505, radius=0.505, shapeName=dyn_tassel_MSTR_capsule, rotation=<rot 0 -90 0>}]
}
constraints [
 {bone1Name=dyn_tassel_01, bone2Name=dyn_tassel_MSTR, swing1LimitDegrees=10, swing2LimitDegrees=30, twistLimitDegrees=5}
 {bone1Name=dyn_tassel_02, bone2Name=dyn_tassel_01, swing1LimitDegrees=35, swing2LimitDegrees=35, twistLimitDegrees=5}
 {bone1Name=dyn_tassel_03, bone2Name=dyn_tassel_02, swing1LimitDegrees=35, swing2LimitDegrees=35, twistLimitDegrees=5}
 {bone1Name=dyn_tassel_04, bone2Name=dyn_tassel_03, swing1LimitDegrees=35, swing2LimitDegrees=35, twistLimitDegrees=5}
 {bone1Name=dyn_tassel_05, bone2Name=dyn_tassel_04, swing1LimitDegrees=35, swing2LimitDegrees=35, twistLimitDegrees=5}
 {bone1Name=dyn_tassel_06, bone2Name=dyn_tassel_05, swing1LimitDegrees=35, swing2LimitDegrees=35, twistLimitDegrees=5}
]
