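# Animation Blueprint for the EvokeFind pickaxe that adds secondary physics motion to its six-bone tassel.
# Bool blends on front-end, on-pawn and rigid-body state select between RigidBody nodes and differently tuned AnimDynamics chains, most with additive bend offsets.
# A rotation-limited Trail controller follows the blends; one extra AnimDynamics chain is unconnected.
# The event graph flags when both physics modes are off and then sets trail relaxation speed by on-pawn state.
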AnimBlueprint /BRCosmetics/Weapons/FORT_Melee/Pickaxe_EvokeFind/Meshes/EvokeFind_Axe_AnimBP
target_skeleton=&/EvokeFind_Axe_Skeleton
graphs [
 AnimGraph {nodes [
   n1 = Anim Dynamics - Bone: dyn_tassel_01(Alpha=1, ExternalForce="0,0,0") {
    angularDampingOverride=1
    angularSpringConstant=850
    bAngularSpring=true
    bOverrideAngularDamping=true
    bUseGravityOverride=true
    alphaBoolBlend {blendOption=Linear}
    boundBone {boneName=dyn_tassel_01}
    componentAppliedLinearAccClamp=<1000 1000 1000>
    componentLinearAccScale=<0.6 0.6 0.6>
    componentLinearVelScale=<0.7 0.7 0.7>
    gravityOverride {z=-900}
    physicsBodyDefinitions [{
      boundBone {boneName=dyn_tassel_01}
      boxExtents=<13 10 10>
      constraintSetup {angularLimitsMax=<1 30 30>, angularLimitsMin=<-1 -30 -30>, angularTarget=<1 0 0>}
      localJointOffset=<1.2 0 0>
     }]
    relativeSpaceBone {boneName=dyn_tassel_02}
   }
   n2 = Anim Dynamics - Bone: dyn_tassel_02(ComponentPose=@n1, Alpha=1, ExternalForce="0,0,0") {
    angularDampingOverride=1
    angularSpringConstant=850
    bAngularSpring=true
    bOverrideAngularDamping=true
    bUseGravityOverride=true
    alphaBoolBlend {blendOption=Linear}
    boundBone {boneName=dyn_tassel_02}
    componentAppliedLinearAccClamp=<1000 1000 1000>
    componentLinearAccScale=<0.6 0.6 0.6>
    componentLinearVelScale=<0.7 0.7 0.7>
    gravityOverride {z=-900}
    physicsBodyDefinitions [{
      boundBone {boneName=dyn_tassel_02}
      boxExtents=<13 10 10>
      constraintSetup {angularLimitsMax=<1 30 30>, angularLimitsMin=<-1 -30 -30>, angularTarget=<1 0 0>}
      localJointOffset=<1.2 0 0>
     }]
    relativeSpaceBone {boneName=dyn_tassel_02}
   }
   n3 = Anim Dynamics - Bone: dyn_tassel_03(ComponentPose=@n2, Alpha=1, ExternalForce="0,0,0") {
    angularDampingOverride=1
    angularSpringConstant=850
    bAngularSpring=true
    bOverrideAngularDamping=true
    bUseGravityOverride=true
    alphaBoolBlend {blendOption=Linear}
    boundBone {boneName=dyn_tassel_03}
    componentAppliedLinearAccClamp=<1000 1000 1000>
    componentLinearAccScale=<0.6 0.6 0.6>
    componentLinearVelScale=<0.7 0.7 0.7>
    gravityOverride {z=-900}
    physicsBodyDefinitions [{
      boundBone {boneName=dyn_tassel_03}
      boxExtents=<13 10 10>
      constraintSetup {angularLimitsMax=<1 30 30>, angularLimitsMin=<-1 -30 -30>, angularTarget=<1 0 0>}
      localJointOffset=<1.2 0 0>
     }]
    relativeSpaceBone {boneName=dyn_tassel_02}
   }
   n4 = Anim Dynamics - Bone: dyn_tassel_04(ComponentPose=@n3, Alpha=1, ExternalForce="0,0,0") {
    angularDampingOverride=1
    angularSpringConstant=850
    bAngularSpring=true
    bOverrideAngularDamping=true
    bUseGravityOverride=true
    alphaBoolBlend {blendOption=Linear}
    boundBone {boneName=dyn_tassel_04}
    componentAppliedLinearAccClamp=<1000 1000 1000>
    componentLinearAccScale=<0.6 0.6 0.6>
    componentLinearVelScale=<0.7 0.7 0.7>
    gravityOverride {z=-900}
    physicsBodyDefinitions [{
      boundBone {boneName=dyn_tassel_04}
      boxExtents=<13 10 10>
      constraintSetup {angularLimitsMax=<1 30 30>, angularLimitsMin=<-1 -30 -30>, angularTarget=<1 0 0>}
      localJointOffset=<1.2 0 0>
     }]
    relativeSpaceBone {boneName=dyn_tassel_02}
   }
   n5 = Anim Dynamics - Bone: dyn_tassel_05(ComponentPose=@n4, Alpha=1, ExternalForce="0,0,0") {
    angularDampingOverride=1
    angularSpringConstant=850
    bAngularSpring=true
    bOverrideAngularDamping=true
    bUseGravityOverride=true
    alphaBoolBlend {blendOption=Linear}
    boundBone {boneName=dyn_tassel_05}
    componentAppliedLinearAccClamp=<1000 1000 1000>
    componentLinearAccScale=<0.6 0.6 0.6>
    componentLinearVelScale=<0.7 0.7 0.7>
    gravityOverride {z=-900}
    physicsBodyDefinitions [{
      boundBone {boneName=dyn_tassel_05}
      boxExtents=<13 10 10>
      constraintSetup {angularLimitsMax=<1 30 30>, angularLimitsMin=<-1 -30 -30>, angularTarget=<1 0 0>}
      localJointOffset=<1.2 0 0>
     }]
    relativeSpaceBone {boneName=dyn_tassel_02}
   }
   Anim Dynamics - Bone: dyn_tassel_06(ComponentPose=@n5, Alpha=1, ExternalForce="0,0,0") {
    angularDampingOverride=1
    angularSpringConstant=850
    bAngularSpring=true
    bOverrideAngularDamping=true
    bUseGravityOverride=true
    alphaBoolBlend {blendOption=Linear}
    boundBone {boneName=dyn_tassel_06}
    componentAppliedLinearAccClamp=<1000 1000 1000>
    componentLinearAccScale=<0.6 0.6 0.6>
    componentLinearVelScale=<0.7 0.7 0.7>
    gravityOverride {z=-900}
    physicsBodyDefinitions [{
      boundBone {boneName=dyn_tassel_06}
      boxExtents=<13 10 10>
      constraintSetup {angularLimitsMax=<1 30 30>, angularLimitsMin=<-1 -30 -30>, angularTarget=<1 0 0>}
      localJointOffset=<1.2 0 0>
     }]
    relativeSpaceBone {boneName=dyn_tassel_02}
   }
   n6 = Anim Dynamics - Bone: dyn_tassel_01(ComponentPose=Local To Component(LocalPose=Input Pose()), Alpha=1, ExternalForce="0,0,0") {
    angularDampingOverride=0.9998
    angularSpringConstant=850
    bAngularSpring=true
    bOverrideAngularDamping=true
    bUseGravityOverride=true
    alphaBoolBlend {blendOption=Linear}
    boundBone {boneName=dyn_tassel_01}
    componentAppliedLinearAccClamp=<1000 1000 1000>
    componentLinearAccScale=<0.6 0.6 0.6>
    componentLinearVelScale=<0.7 0.7 0.7>
    gravityOverride {z=-900}
    physicsBodyDefinitions [{
      boundBone {boneName=dyn_tassel_01}
      boxExtents=<12 12 12>
      constraintSetup {angularLimitsMax=<1 30 10>, angularLimitsMin=<-1 -30 -10>, angularTarget=<1 0 0>}
      localJointOffset=<1.3 0 0>
     }]
    relativeSpaceBone {boneName=dyn_tassel_02}
   }
   n7 = Anim Dynamics - Bone: dyn_tassel_02(ComponentPose=@n6, Alpha=1, ExternalForce="0,0,0") {
    angularDampingOverride=0.9998
    angularSpringConstant=850
    bAngularSpring=true
    bOverrideAngularDamping=true
    bUseGravityOverride=true
    alphaBoolBlend {blendOption=Linear}
    boundBone {boneName=dyn_tassel_02}
    componentAppliedLinearAccClamp=<1000 1000 1000>
    componentLinearAccScale=<0.6 0.6 0.6>
    componentLinearVelScale=<0.7 0.7 0.7>
    gravityOverride {z=-900}
    physicsBodyDefinitions [{
      boundBone {boneName=dyn_tassel_02}
      boxExtents=<11 11 11>
      constraintSetup {angularLimitsMax=<1 30 10>, angularLimitsMin=<-1 -30 -10>, angularTarget=<1 0 0>}
      localJointOffset=<1.3 0 0>
     }]
    relativeSpaceBone {boneName=dyn_tassel_02}
   }
   n8 = Anim Dynamics - Bone: dyn_tassel_03(ComponentPose=@n7, Alpha=1, ExternalForce="0,0,0") {
    angularDampingOverride=0.9998
    angularSpringConstant=850
    bAngularSpring=true
    bOverrideAngularDamping=true
    bUseGravityOverride=true
    alphaBoolBlend {blendOption=Linear}
    boundBone {boneName=dyn_tassel_03}
    componentAppliedLinearAccClamp=<1000 1000 1000>
    componentLinearAccScale=<0.6 0.6 0.6>
    componentLinearVelScale=<0.7 0.7 0.7>
    gravityOverride {z=-900}
    physicsBodyDefinitions [{
      boundBone {boneName=dyn_tassel_03}
      boxExtents=<10 10 10>
      constraintSetup {angularLimitsMax=<1 30 10>, angularLimitsMin=<-1 -30 -10>, angularTarget=<1 0 0>}
      localJointOffset=<1.3 0 0>
     }]
    relativeSpaceBone {boneName=dyn_tassel_02}
   }
   n9 = Anim Dynamics - Bone: dyn_tassel_04(ComponentPose=@n8, Alpha=1, ExternalForce="0,0,0") {
    angularDampingOverride=0.9998
    angularSpringConstant=850
    bAngularSpring=true
    bOverrideAngularDamping=true
    bUseGravityOverride=true
    alphaBoolBlend {blendOption=Linear}
    boundBone {boneName=dyn_tassel_04}
    componentAppliedLinearAccClamp=<1000 1000 1000>
    componentLinearAccScale=<0.6 0.6 0.6>
    componentLinearVelScale=<0.7 0.7 0.7>
    gravityOverride {z=-900}
    physicsBodyDefinitions [{
      boundBone {boneName=dyn_tassel_04}
      boxExtents=<8 8 8>
      constraintSetup {angularLimitsMax=<1 30 10>, angularLimitsMin=<-1 -30 -10>, angularTarget=<1 0 0>}
      localJointOffset=<1.3 0 0>
     }]
    relativeSpaceBone {boneName=dyn_tassel_02}
   }
   n10 = Anim Dynamics - Bone: dyn_tassel_05(ComponentPose=@n9, Alpha=1, ExternalForce="0,0,0") {
    angularDampingOverride=0.9998
    angularSpringConstant=850
    bAngularSpring=true
    bOverrideAngularDamping=true
    bUseGravityOverride=true
    alphaBoolBlend {blendOption=Linear}
    boundBone {boneName=dyn_tassel_05}
    componentAppliedLinearAccClamp=<1000 1000 1000>
    componentLinearAccScale=<0.6 0.6 0.6>
    componentLinearVelScale=<0.7 0.7 0.7>
    gravityOverride {z=-900}
    physicsBodyDefinitions [{
      boundBone {boneName=dyn_tassel_05}
      boxExtents=<7 7 7>
      constraintSetup {angularLimitsMax=<1 30 10>, angularLimitsMin=<-1 -30 -10>, angularTarget=<1 0 0>}
      localJointOffset=<1.3 0 0>
     }]
    relativeSpaceBone {boneName=dyn_tassel_02}
   }
   n11 = Anim Dynamics - Bone: dyn_tassel_06(ComponentPose=@n10, Alpha=1, ExternalForce="0,0,0") {
    angularDampingOverride=0.9998
    angularSpringConstant=850
    bAngularSpring=true
    bOverrideAngularDamping=true
    bUseGravityOverride=true
    alphaBoolBlend {blendOption=Linear}
    boundBone {boneName=dyn_tassel_06}
    componentAppliedLinearAccClamp=<1000 1000 1000>
    componentLinearAccScale=<0.6 0.6 0.6>
    componentLinearVelScale=<0.7 0.7 0.7>
    gravityOverride {z=-900}
    physicsBodyDefinitions [{
      boundBone {boneName=dyn_tassel_06}
      boxExtents=<5 5 5>
      constraintSetup {angularLimitsMax=<1 30 10>, angularLimitsMin=<-1 -30 -10>, angularTarget=<1 0 0>}
      localJointOffset=<1.3 0 0>
     }]
    relativeSpaceBone {boneName=dyn_tassel_02}
   }
   n12 = "Transform (Modify) Bone - Bone: dyn_tassel_01"!(ComponentPose=@n11, bAlphaBoolEnabled:bool=True, Rotation="-30,0,0") -> (Pose) {
    alphaInputType=Bool
    rotationMode=BMM_Additive
    alphaBoolBlend {blendOption=Linear}
    boneToModify {boneName=dyn_tassel_01}
    rotation {pitch=-30}
   }
   n13 = "Transform (Modify) Bone - Bone: dyn_tassel_03"!(ComponentPose=@n12, bAlphaBoolEnabled:bool=True, Rotation="-25,0,0") -> (Pose) {
    alphaInputType=Bool
    rotationMode=BMM_Additive
    alphaBoolBlend {blendOption=Linear}
    boneToModify {boneName=dyn_tassel_03}
    rotation {pitch=-25}
   }
   n14 = "Transform (Modify) Bone - Bone: dyn_tassel_05"!(ComponentPose=@n13, bAlphaBoolEnabled:bool=True, Rotation="-25,0,0") -> (Pose) {
    alphaInputType=Bool
    rotationMode=BMM_Additive
    alphaBoolBlend {blendOption=Linear}
    boneToModify {boneName=dyn_tassel_05}
    rotation {pitch=-25}
   }
   n15 = RigidBody(ComponentPose=Local To Component(LocalPose=Input Pose() {name=InPose_3})) {
    bOverrideWorldGravity=true
    cachedBoundsScale=1
    overridePhysicsAsset=&/EvokeFind_Axe_Physics
    alphaBoolBlend {blendOption=Linear}
    overrideWorldGravity {z=-900}
    simSpaceSettings {dampingAlpha=5, velocityScaleZ=3, worldAlpha=0.5}
   }
   n16 = Anim Dynamics - Bone: dyn_tassel_01(ComponentPose=Local To Component(LocalPose=Input Pose() {name=InPose_0}), Alpha=1, ExternalForce="0,0,0") {
    angularDampingOverride=0.998
    bOverrideAngularDamping=true
    bUseGravityOverride=true
    alphaBoolBlend {blendOption=Linear}
    boundBone {boneName=dyn_tassel_01}
    gravityOverride {z=-800}
    physicsBodyDefinitions [{
      boundBone {boneName=dyn_tassel_01}
      boxExtents=<15 10 10>
      constraintSetup {angularLimitsMax=<1 30 30>, angularLimitsMin=<-1 -30 -30>, angularTarget=<0 0 0>}
      localJointOffset=<1 0 0>
     }]
    relativeSpaceBone {boneName=dyn_tassel_02}
    simSpaceSettings {maxAngularAcceleration=100, maxAngularVelocity=20, simSpaceAngularAlpha=1}
   }
   n17 = Anim Dynamics - Bone: dyn_tassel_02(ComponentPose=@n16, Alpha=1, ExternalForce="0,0,0") {
    angularDampingOverride=0.998
    bOverrideAngularDamping=true
    bUseGravityOverride=true
    alphaBoolBlend {blendOption=Linear}
    boundBone {boneName=dyn_tassel_02}
    gravityOverride {z=-800}
    physicsBodyDefinitions [{
      boundBone {boneName=dyn_tassel_02}
      boxExtents=<14 10 10>
      constraintSetup {angularLimitsMax=<1 30 30>, angularLimitsMin=<-1 -30 -30>, angularTarget=<0 0 0>}
      localJointOffset=<1 0 0>
     }]
    relativeSpaceBone {boneName=dyn_tassel_02}
    simSpaceSettings {maxAngularAcceleration=100, maxAngularVelocity=20, simSpaceAngularAlpha=1}
   }
   n18 = Anim Dynamics - Bone: dyn_tassel_03(ComponentPose=@n17, Alpha=1, ExternalForce="0,0,0") {
    angularDampingOverride=0.998
    bOverrideAngularDamping=true
    bUseGravityOverride=true
    alphaBoolBlend {blendOption=Linear}
    boundBone {boneName=dyn_tassel_03}
    gravityOverride {z=-800}
    physicsBodyDefinitions [{
      boundBone {boneName=dyn_tassel_03}
      boxExtents=<13 10 10>
      constraintSetup {angularLimitsMax=<1 30 30>, angularLimitsMin=<-1 -30 -30>, angularTarget=<0 0 0>}
      localJointOffset=<1 0 0>
     }]
    relativeSpaceBone {boneName=dyn_tassel_02}
    simSpaceSettings {maxAngularAcceleration=100, maxAngularVelocity=20, simSpaceAngularAlpha=1}
   }
   n19 = Anim Dynamics - Bone: dyn_tassel_04(ComponentPose=@n18, Alpha=1, ExternalForce="0,0,0") {
    angularDampingOverride=0.998
    bOverrideAngularDamping=true
    bUseGravityOverride=true
    alphaBoolBlend {blendOption=Linear}
    boundBone {boneName=dyn_tassel_04}
    gravityOverride {z=-800}
    physicsBodyDefinitions [{
      boundBone {boneName=dyn_tassel_04}
      boxExtents=<12 10 10>
      constraintSetup {angularLimitsMax=<1 30 30>, angularLimitsMin=<-1 -30 -30>, angularTarget=<0 0 0>}
      localJointOffset=<1 0 0>
     }]
    relativeSpaceBone {boneName=dyn_tassel_02}
    simSpaceSettings {maxAngularAcceleration=100, maxAngularVelocity=20, simSpaceAngularAlpha=1}
   }
   n20 = Anim Dynamics - Bone: dyn_tassel_05(ComponentPose=@n19, Alpha=1, ExternalForce="0,0,0") {
    angularDampingOverride=0.998
    bOverrideAngularDamping=true
    bUseGravityOverride=true
    alphaBoolBlend {blendOption=Linear}
    boundBone {boneName=dyn_tassel_05}
    gravityOverride {z=-800}
    physicsBodyDefinitions [{
      boundBone {boneName=dyn_tassel_05}
      boxExtents=<11 10 10>
      constraintSetup {angularLimitsMax=<1 30 30>, angularLimitsMin=<-1 -30 -30>, angularTarget=<0 0 0>}
      localJointOffset=<1 0 0>
     }]
    relativeSpaceBone {boneName=dyn_tassel_02}
    simSpaceSettings {maxAngularAcceleration=100, maxAngularVelocity=20, simSpaceAngularAlpha=1}
   }
   n21 = Anim Dynamics - Bone: dyn_tassel_06(ComponentPose=@n20, Alpha=1, ExternalForce="0,0,0") {
    angularDampingOverride=0.998
    bOverrideAngularDamping=true
    bUseGravityOverride=true
    alphaBoolBlend {blendOption=Linear}
    boundBone {boneName=dyn_tassel_06}
    gravityOverride {z=-800}
    physicsBodyDefinitions [{
      boundBone {boneName=dyn_tassel_06}
      boxExtents=<10 10 10>
      constraintSetup {angularLimitsMax=<1 30 30>, angularLimitsMin=<-1 -30 -30>, angularTarget=<0 0 0>}
      localJointOffset=<1 0 0>
     }]
    relativeSpaceBone {boneName=dyn_tassel_02}
    simSpaceSettings {maxAngularAcceleration=100, maxAngularVelocity=20, simSpaceAngularAlpha=1}
   }
   n22 = "Transform (Modify) Bone - Bone: dyn_tassel_01"(ComponentPose=@n21, bAlphaBoolEnabled=True, Rotation="-32.5,0,0") {
    alphaInputType=Bool
    rotationMode=BMM_Additive
    alphaBoolBlend {blendOption=Linear}
    boneToModify {boneName=dyn_tassel_01}
    rotation {pitch=-32.5}
   }
   n23 = Blend Poses by bool(
    BlendPose_0=Blend Poses by bool(BlendPose_0=Component To Local(ComponentPose=@n15), BlendPose_1=Component To Local(ComponentPose=@n14), bActiveValue=Get bRigidBodyEnabled()) {blendType=Linear}
    BlendPose_1=Component To Local(ComponentPose=@n22)
    bActiveValue=Get bIsOnPawn() -> (bIsOnPawn:bool)
   ) {blendType=Linear}
   n24 = RigidBody(ComponentPose=Local To Component(LocalPose=Input Pose() {name=InPose_1})) {
    bOverrideWorldGravity=true
    overridePhysicsAsset=&EvokeFind_Axe_Physics
    simulationSpace=BaseBoneSpace
    alphaBoolBlend {blendOption=Linear}
    baseBoneRef {boneName=Melee_TwoHanded_Handle}
    overrideWorldGravity {z=-800}
    simSpaceSettings {
     dampingAlpha=0
     maxAngularAcceleration=10
     maxAngularVelocity=2
     maxLinearAcceleration=1200
     maxLinearVelocity=300
     velocityScaleZ=0.15
     worldAlpha=1
    }
   }
   n25 = Component To Local(ComponentPose=@n24)
   n26 = Anim Dynamics - Bone: dyn_tassel_01(ComponentPose=Local To Component(LocalPose=Input Pose() {name=InPose_2}), Alpha=1, ExternalForce="0,0,0") {
    angularDampingOverride=0.99
    angularSpringConstant=100
    bAngularSpring=true
    bOverrideAngularDamping=true
    bUseGravityOverride=true
    alphaBoolBlend {blendOption=Linear}
    boundBone {boneName=dyn_tassel_01}
    componentAppliedLinearAccClamp=<800 800 1000>
    componentLinearAccScale=<0.5 0.5 0.5>
    componentLinearVelScale=<0.7 0.7 0.7>
    gravityOverride {z=-980}
    physicsBodyDefinitions [{
      boundBone {boneName=dyn_tassel_01}
      boxExtents=<14 14 13>
      constraintSetup {angularLimitsMax=<1 30 30>, angularLimitsMin=<-1 -30 -30>, angularTarget=<1 0 0>}
      localJointOffset=<1.5 0 0>
     }]
    relativeSpaceBone {boneName=dyn_tassel_02}
   }
   n27 = Anim Dynamics - Bone: dyn_tassel_02(ComponentPose=@n26, Alpha=1, ExternalForce="0,0,0") {
    angularDampingOverride=0.99
    angularSpringConstant=100
    bAngularSpring=true
    bOverrideAngularDamping=true
    bUseGravityOverride=true
    alphaBoolBlend {blendOption=Linear}
    boundBone {boneName=dyn_tassel_02}
    componentAppliedLinearAccClamp=<800 800 1000>
    componentLinearAccScale=<0.5 0.5 0.5>
    componentLinearVelScale=<0.7 0.7 0.7>
    gravityOverride {z=-980}
    physicsBodyDefinitions [{
      boundBone {boneName=dyn_tassel_02}
      boxExtents=<14 13 13>
      constraintSetup {angularLimitsMax=<1 30 30>, angularLimitsMin=<-1 -30 -30>, angularTarget=<1 0 0>}
      localJointOffset=<1.5 0 0>
     }]
    relativeSpaceBone {boneName=dyn_tassel_02}
   }
   n28 = Anim Dynamics - Bone: dyn_tassel_03(ComponentPose=@n27, Alpha=1, ExternalForce="0,0,0") {
    angularDampingOverride=0.99
    angularSpringConstant=100
    bAngularSpring=true
    bOverrideAngularDamping=true
    bUseGravityOverride=true
    alphaBoolBlend {blendOption=Linear}
    boundBone {boneName=dyn_tassel_03}
    componentAppliedLinearAccClamp=<800 800 1000>
    componentLinearAccScale=<0.5 0.5 0.5>
    componentLinearVelScale=<0.7 0.7 0.7>
    gravityOverride {z=-980}
    physicsBodyDefinitions [{
      boundBone {boneName=dyn_tassel_03}
      boxExtents=<13 13 13>
      constraintSetup {angularLimitsMax=<1 30 30>, angularLimitsMin=<-1 -30 -30>, angularTarget=<1 0 0>}
      localJointOffset=<1.5 0 0>
     }]
    relativeSpaceBone {boneName=dyn_tassel_02}
   }
   n29 = Anim Dynamics - Bone: dyn_tassel_04(ComponentPose=@n28, Alpha=1, ExternalForce="0,0,0") {
    angularDampingOverride=0.99
    angularSpringConstant=100
    bAngularSpring=true
    bOverrideAngularDamping=true
    bUseGravityOverride=true
    alphaBoolBlend {blendOption=Linear}
    boundBone {boneName=dyn_tassel_04}
    componentAppliedLinearAccClamp=<800 800 1000>
    componentLinearAccScale=<0.5 0.5 0.5>
    componentLinearVelScale=<0.7 0.7 0.7>
    gravityOverride {z=-980}
    physicsBodyDefinitions [{
      boundBone {boneName=dyn_tassel_04}
      boxExtents=<12 12 12>
      constraintSetup {angularLimitsMax=<1 30 30>, angularLimitsMin=<-1 -30 -30>, angularTarget=<1 0 0>}
      localJointOffset=<1.5 0 0>
     }]
    relativeSpaceBone {boneName=dyn_tassel_02}
   }
   n30 = Anim Dynamics - Bone: dyn_tassel_05(ComponentPose=@n29, Alpha=1, ExternalForce="0,0,0") {
    angularDampingOverride=0.99
    angularSpringConstant=100
    bAngularSpring=true
    bOverrideAngularDamping=true
    bUseGravityOverride=true
    alphaBoolBlend {blendOption=Linear}
    boundBone {boneName=dyn_tassel_05}
    componentAppliedLinearAccClamp=<800 800 1000>
    componentLinearAccScale=<0.5 0.5 0.5>
    componentLinearVelScale=<0.7 0.7 0.7>
    gravityOverride {z=-980}
    physicsBodyDefinitions [{
      boundBone {boneName=dyn_tassel_05}
      boxExtents=<11 11 11>
      constraintSetup {angularLimitsMax=<1 30 30>, angularLimitsMin=<-1 -30 -30>, angularTarget=<1 0 0>}
      localJointOffset=<1.5 0 0>
     }]
    relativeSpaceBone {boneName=dyn_tassel_02}
   }
   n31 = Anim Dynamics - Bone: dyn_tassel_06(ComponentPose=@n30, Alpha=1, ExternalForce="0,0,0") {
    angularDampingOverride=0.99
    angularSpringConstant=100
    bAngularSpring=true
    bOverrideAngularDamping=true
    bUseGravityOverride=true
    alphaBoolBlend {blendOption=Linear}
    boundBone {boneName=dyn_tassel_06}
    componentAppliedLinearAccClamp=<800 800 1000>
    componentLinearAccScale=<0.5 0.5 0.5>
    componentLinearVelScale=<0.7 0.7 0.7>
    gravityOverride {z=-980}
    physicsBodyDefinitions [{
      boundBone {boneName=dyn_tassel_06}
      boxExtents=<10 10 10>
      constraintSetup {angularLimitsMax=<1 30 30>, angularLimitsMin=<-1 -30 -30>, angularTarget=<1 0 0>}
      localJointOffset=<1.5 0 0>
     }]
    relativeSpaceBone {boneName=dyn_tassel_02}
   }
   n32 = Component To Local(ComponentPose=@n31)
   n33 = Get bRigidBodyEnabled()
   n34 = "Transform (Modify) Bone - Bone: dyn_tassel_01"(
    ComponentPose=Local To Component(LocalPose=Blend Poses by bool(BlendPose_0=@n25, BlendPose_1=@n32, bActiveValue=@n33) {blendType=Linear})
    bAlphaBoolEnabled=True
    Rotation="-30,0,0"
   ) {
    alphaInputType=Bool
    rotationMode=BMM_Additive
    alphaBoolBlend {blendOption=Linear}
    boneToModify {boneName=dyn_tassel_01}
    rotation {pitch=-30}
   }
   n35 = "Transform (Modify) Bone - Bone: dyn_tassel_03"(ComponentPose=@n34, bAlphaBoolEnabled=True, Rotation="-25,0,0") {
    alphaInputType=Bool
    rotationMode=BMM_Additive
    alphaBoolBlend {blendOption=Linear}
    boneToModify {boneName=dyn_tassel_03}
    rotation {pitch=-25}
   }
   n36 = "Transform (Modify) Bone - Bone: dyn_tassel_05"(ComponentPose=@n35, bAlphaBoolEnabled=True, Rotation="-25,0,0") {
    alphaInputType=Bool
    rotationMode=BMM_Additive
    alphaBoolBlend {blendOption=Linear}
    boneToModify {boneName=dyn_tassel_05}
    rotation {pitch=-25}
   }
   n37 = Component To Local(ComponentPose=@n36)
   n38 = Get bIsInFrontEnd()
   n39 = Trail controller - Bone: dyn_tassel_06!(
    ComponentPose=Local To Component(LocalPose=Blend Poses by bool(BlendPose_0=@n23, BlendPose_1=@n37, bActiveValue=@n38) {blendType=Linear})
    bAlphaBoolEnabled:bool=True
    RelaxationSpeedScale:float=3.5
   ) -> (Pose) {
    alphaInputType=Bool
    bLimitRotation=true
    chainLength=6
    maxDeltaTime=0.0333
    relaxationSpeedScale=3.5
    rotationLimits [
     {limitMax=<30 30 30>, limitMin=<-30 -30 -30>}
     {limitMax=<30 30 30>, limitMin=<-30 -30 -30>}
     {limitMax=<30 30 30>, limitMin=<-30 -30 -30>}
     {limitMax=<30 30 30>, limitMin=<-30 -30 -30>}
     {limitMax=<30 30 30>, limitMin=<-30 -30 -30>}
     {limitMax=<30 30 30>, limitMin=<-30 -30 -30>}
    ]
    rotationOffsets [<0 0 0>, <0 0 0>, <0 0 0>, <0 0 0>, <0 0 0>, <0 0 0>]
    trailBone {boneName=dyn_tassel_06}
    trailRelaxationSpeed {editorCurveData {keys [<at 0 10 RCIM_Cubic>,<at 1 5 RCIM_Cubic>]}}
   }
   Output Pose(Result=Component To Local(ComponentPose=@n39)) {layerGroup=None}
  ]}
 EventGraph {nodes [
   Event BlueprintUpdateAnimation()
   > n1 = Sequence!(execute) -> (then_0, then_1, then_2:Exec)
   > Branch(Condition=Get Is NOT AnimDynamic OnPawn() -> (Is NOT AnimDynamic OnPawn:bool))
   > Set RelaxationSpeedScale(RelaxationSpeedScale:double=3.5) -> (Output_Get:double)
   Branch(execute=@n1.then_1, Condition=Get Is NOT AnimDynamic NOT OnPawn() -> (Is NOT AnimDynamic NOT OnPawn:bool))
   > Set RelaxationSpeedScale(RelaxationSpeedScale=2.5)
   n2 = AND Boolean(A=NOT Boolean(A=Get bAnimDynamicsEnabled() -> (bAnimDynamicsEnabled:bool)), B=NOT Boolean(A=Get bRigidBodyEnabled()))
   Set IsNOTAnimDynamic(execute=@n1.then_2, IsNOTAnimDynamic:bool=@n2) -> (Output_Get:bool)
   n3 = Get bIsOnPawn()
   > Set Is NOT AnimDynamic OnPawn(Is NOT AnimDynamic OnPawn:bool=AND Boolean(A=@n2, B=@n3)) -> (Output_Get:bool)
   > Set Is NOT AnimDynamic NOT OnPawn(Is NOT AnimDynamic NOT OnPawn:bool=AND Boolean(A=@n2, B=NOT Boolean(A=@n3))) -> (Output_Get:bool)
  ]}
]
variables [
 Is NOT AnimDynamic NOT OnPawn {type=boolean}
 Is NOT AnimDynamic OnPawn {type=boolean}
 IsNOTAnimDynamic {type=boolean}
 RelaxationSpeedScale {type=number}
]
